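; PCG debug subgraph that gives each incoming data set its own color, as its name indicates.
; It writes a dataIndex on each input, extracts it and merges the results into one attribute set.
; That value is cast to a Vector, randomized with Attribute Noise, then partitioned by $Index into one entry per data set.
; Copy Attributes writes these random vectors back to the original inputs as $color.
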(graph OneColorPerDataSet
  (bind WriteDataIndex_1 (Write%20Data%20Index :bAddAttribute true :bAddTag false :IndexAttribute dataIndex))
  (bind ExtractAttribute_1 (Extract%20Attribute%20at%20Index :bForceOutputAttributeToBeInElementsDomain true))
  (bind MergeAttributes_1 (Merge%20Attributes))
  (bind AttributeCast_1 (Attribute%20Cast :OutputType Vector))
  (bind AttributeNoise_1 (Attribute%20Noise :InputSource dataIndex))
  (bind AttributePartition_1 (Attribute%20Partition :PartitionAttributeSelectors [$Index]))
  (bind CopyAttributes_1 (Copy%20Attributes :InputSource @Last :OutputTarget $color))
  (connect MergeAttributes_1 Out AttributeCast_1 In)
  (connect AttributeCast_1 Out AttributeNoise_1 In)
  (connect AttributeNoise_1 Out AttributePartition_1 In)
  (connect AttributePartition_1 Out CopyAttributes_1 Source)
  (connect DefaultInputNode In CopyAttributes_1 Target)
  (connect WriteDataIndex_1 Out ExtractAttribute_1 In)
  (connect ExtractAttribute_1 Out MergeAttributes_1 In)
  (connect DefaultInputNode In WriteDataIndex_1 In)
  (output Out CopyAttributes_1))
